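# Looping ambient spot sound cue for the shore side of the WildEstate purple water, in the outdoor ambient sound class.
# A looping purple-water layer wave feeds a modulator that fixes pitch slightly below normal.
# An attenuation node at the root applies the shared purple-water shore attenuation asset.
# The cue also carries its own attenuation override values and single-instance concurrency settings that prevent new plays.
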SoundCue /WildEstate/Audio/Spots/Spot_WildEstate_PurpleWater_ShoreSide_Cue
root=Attenuation_9
nodes [
 {id=Attenuation_9, children [Modulator_2], properties {attenuationSettings=&/WildEstate_PurpleWater_Shore_Att}}
 {id=Modulator_2, children [WavePlayer_3], properties {pitchMax=0.9, pitchMin=0.9}}
 {id=WavePlayer_3, properties {bLooping=true, soundWaveAssetPtr=&/Spot_Purple_Water_Layer_Loop}}
]
properties {
 maxDistance=2833
 soundClassObject=&Fort_Ambient_Outdoors_Athena
 attenuationOverrides {falloffDistance=2000, nonSpatializedRadiusStart=700, stereoSpread=1000, attenuationShapeExtents {x=925}}
 concurrencyOverrides {maxCount=1, resolutionRule=PreventNew}
}
